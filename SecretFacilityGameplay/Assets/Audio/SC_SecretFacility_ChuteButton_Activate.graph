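# Sound cue for the Secret Facility chute button activation sound.
# A single wave player feeds the root directly, with no randomization or mixing.
# It routes to the gadgets and items sound class and uses an attenuation asset borrowed from a supply drop ambient loop.
# A concurrency override stops the oldest instance when the limit is reached.
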SoundCue /SecretFacilityGameplay/Assets/Audio/SC_SecretFacility_ChuteButton_Activate
root=WavePlayer_0
nodes [{id=WavePlayer_0, properties {soundWaveAssetPtr=&/SecretFacility_ChuteButton_Activate}}]
properties {
 attenuationSettings=/Game/Athena/Sounds/SupplyDrops/Fort_HiddenStar_Ambient_Loop_Att
 bOverrideConcurrency=true
 maxDistance=2500
 soundClassObject=&Fort_Gadgets_Items
 volumeMultiplier=1
 attenuationOverrides {falloffDistance=1000, attenuationShapeExtents {x=1000}}
 concurrencyOverrides {resolutionRule=StopOldest}
}
