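# Sound cue for the Secret Facility chute doors opening sound.
# A single wave player feeds the root directly, with boosted volume and routing through the Fort_Doors sound class.
# Inline attenuation overrides take precedence over the assigned Fort_Doors_Att asset, setting their own shape extent and falloff.
# A concurrency override stops the oldest instance when the play limit is reached.
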SoundCue /SecretFacilityGameplay/Assets/Audio/SC_SecretFacility_ChuteDoors_Open
root=WavePlayer_1
nodes [{id=WavePlayer_1, properties {soundWaveAssetPtr=&/SecretFacility_ChuteDoors_Open}}]
properties {
 attenuationSettings=/Game/Sounds/Attenuation/Fort_Doors_Att
 bOverrideAttenuation=true
 bOverrideConcurrency=true
 maxDistance=4600
 soundClassObject=&Fort_Doors
 volumeMultiplier=2
 attenuationOverrides {falloffDistance=3000, attenuationShapeExtents {x=1600}}
 concurrencyOverrides {resolutionRule=StopOldest}
}
