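# Sound cue for the Secret Facility vortex, layering a one-shot activation sound with looping ambient, synth and wind beds in the Fort_Ambient_Athena sound class.
# A root Mixer blends the four wave players, each behind its own Attenuation node with overridden falloff, with the synth layer weighted far above the others.
# Synth and wind use non-spatialized capsule shapes of differing size, while activation and ambient keep the default shape with a larger radius.
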SoundCue /SecretFacilityGameplay/Assets/Audio/SC_SecretFacility_Vortex_Loop_Cue
root=Mixer_0
nodes [
 {
  id=Attenuation_0
  children [WavePlayer_0]
  properties {
   bOverrideAttenuation=true
   attenuationOverrides {
    distanceAlgorithm=NaturalSound
    hPFFrequencyAtMax=100
    lPFFrequencyAtMax=808
    lPFRadiusMax=9000
    lPFRadiusMin=500
    occlusionLowPassFilterFrequency=3000
    occlusionTraceChannel=ECC_GameTraceChannel11
    occlusionVolumeAttenuation=0.5
    reverbDistanceMax=5000
    stereoSpread=1000
    attenuationShapeExtents {x=2800}
    customAttenuationCurve {editorCurveData {keys [<at 0 1>,<at 1 0>]}}
   }
  }
 }
 {
  id=Attenuation_2
  children [WavePlayer_1]
  properties {
   bOverrideAttenuation=true
   attenuationOverrides {
    distanceAlgorithm=NaturalSound
    hPFFrequencyAtMax=100
    lPFFrequencyAtMax=808
    lPFRadiusMax=9000
    lPFRadiusMin=500
    occlusionLowPassFilterFrequency=3000
    occlusionTraceChannel=ECC_GameTraceChannel11
    occlusionVolumeAttenuation=0.5
    reverbDistanceMax=5000
    stereoSpread=1000
    attenuationShapeExtents {x=2800}
    customAttenuationCurve {editorCurveData {keys [<at 0 1>,<at 1 0>]}}
   }
  }
 }
 {
  id=Attenuation_4
  children [WavePlayer_3]
  properties {
   bOverrideAttenuation=true
   attenuationOverrides {
    attenuationShape=Capsule
    bSpatialize=false
    coneSphereFalloffDistance=0.2
    dBAttenuationAtMax=-49.44
    distanceAlgorithm=NaturalSound
    falloffDistance=300
    hPFFrequencyAtMax=100
    lPFFrequencyAtMax=808
    lPFRadiusMax=9000
    lPFRadiusMin=500
    occlusionLowPassFilterFrequency=3000
    occlusionTraceChannel=ECC_GameTraceChannel11
    occlusionVolumeAttenuation=0.5
    reverbDistanceMax=5000
    stereoSpread=1000
    attenuationShapeExtents=<5000 400 12.4624>
    customAttenuationCurve {editorCurveData {keys [<at 0 1>,<at 1 0>]}}
   }
  }
 }
 {
  id=Attenuation_7
  children [WavePlayer_6]
  properties {
   bOverrideAttenuation=true
   attenuationOverrides {
    attenuationShape=Capsule
    bSpatialize=false
    coneSphereFalloffDistance=0.2
    dBAttenuationAtMax=-49.44
    distanceAlgorithm=NaturalSound
    falloffDistance=1200
    hPFFrequencyAtMax=100
    lPFFrequencyAtMax=808
    lPFRadiusMax=9000
    lPFRadiusMin=500
    occlusionLowPassFilterFrequency=3000
    occlusionTraceChannel=ECC_GameTraceChannel11
    occlusionVolumeAttenuation=0.5
    reverbDistanceMax=5000
    stereoSpread=1000
    attenuationShapeExtents=<9000 300 12.4624>
    customAttenuationCurve {editorCurveData {keys [<at 0 1>,<at 1 0>]}}
   }
  }
 }
 {id=Mixer_0, children [Attenuation_2, Attenuation_0, Attenuation_4, Attenuation_7], properties {inputVolume [0.75,1.25,10,0.75]}}
 {id=WavePlayer_0, properties {bLooping=true, soundWaveAssetPtr=&/SecretFacility_Vortex_AmbientLoop}}
 {id=WavePlayer_1, properties {soundWaveAssetPtr=&/SecretFacility_Vortex_Activate}}
 {id=WavePlayer_3, properties {bLooping=true, soundWaveAssetPtr=&/SecretFacility_Vortex_Synth_Loop}}
 {id=WavePlayer_6, properties {bLooping=true, soundWaveAssetPtr=&/SecretFacility_Vortex_Wind_Loop}}
]
properties {
 attenuationSettings=&/ATT_SC_SecretFacility_Vortex_Loop_Cue
 bDebug=true
 maxDistance=10456.3105
 soundClassObject=&Fort_Ambient_Athena
 volumeMultiplier=1.5
 attenuationOverrides {falloffDistance=1500, lPFFrequencyAtMax=1200, attenuationShapeExtents {x=667.1792}}
}
