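# Sound cue for the Secret Facility gameplay, named for the moment invisibility is reactivated.
# A single wave player feeds a Modulator at default settings, adding slight random pitch and volume variation to each play.
# The cue boosts volume, uses close-explosion attenuation and the Athena SFX sound class, and overrides concurrency so new instances are refused once the limit is reached.
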SoundCue /SecretFacilityGameplay/Assets/Audio/SecretFacility_Invisibility_Reactivated_Cue
root=Modulator_1
nodes [
 {id=Modulator_1, children [WavePlayer_1]}
 {id=WavePlayer_1, properties {soundWaveAssetPtr=&/SecretFacility_Invisibility_Reactivated_Opt}}
]
properties {
 attenuationSettings=/Game/Sounds/Attenuation/Fort_Explosion_Close_Athena_Att
 bOverrideConcurrency=true
 maxDistance=1499.9875
 soundClassObject=/Game/Sounds/FortSoundClassesAndModes/Fort_Events_SoundClass/Fort_Athena_SFX
 volumeMultiplier=1.5
 concurrencyOverrides {resolutionRule=PreventNew}
}
