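# Sound cue for a Secret Facility warning, playing the invisibility siren wave through a single wave player at boosted volume.
# It routes to the traps sound class and references the shared far grenade explosion attenuation asset for spatial falloff.
# Its own concurrency override caps how many instances play at once.
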SoundCue /SecretFacilityGameplay/Assets/Audio/SecretFacility_Warning_Cue
root=WavePlayer_0
nodes [{id=WavePlayer_0, properties {soundWaveAssetPtr=&/SecretFacility_Invisibility_Siren}}]
properties {
 attenuationSettings=/Game/Sounds/Attenuation/Fort_Explosion_Far_Grenade_Att
 bOverrideConcurrency=true
 maxDistance=10500
 soundClassObject=&Fort_Traps
 volumeMultiplier=1.5
 attenuationOverrides {falloffDistance=1000, attenuationShapeExtents {x=1000}}
 concurrencyOverrides {maxCount=4}
}
