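# Additive unlit flare material for Niagara mesh particles, apparently for spline energy in the Secret Facility invisibility effect.
# A soft flare texture blends low and high colors through a power curve, tinted by particle color and passed through shared emissive adjustments.
# The Sprite function turns geometry into camera-facing sprites, sized by particle alpha and a dynamic parameter.
# Opacity combines particle alpha and flare shape with depth fade.
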Material /SecretFacilityGameplay/Assets/FX/Materials/M_SecretFacilityInvisibility_Energy_SplineFlare
graph {
 properties {
  bUsedWithNiagaraMeshParticles=true
  blendMode=BLEND_Additive
  pixelDepthOffsetMode=PDOM_Legacy
  shadingModel=MSM_Unlit
  usedShadingModels=Unlit
 }
 nodes [
  n1 = ParticleColor()
  n2 = TextureSample() {texture=/Game/Effects/Fort_Effects/Textures/Flares/T_Flare_Soft}
  DepthFade(Opacity=Multiply(A=@n1.A, B=@n2.R)) {fadeDistanceDefault=20}
  n3 = ScalarParameter() {defaultValue=12, parameterName=SpriteSize}
  MaterialFunctionCall!(
   Center Location=WorldPosition().XYZ
   0-1 UVs=TextureCoordinate()
   XY scale=Multiply(A=LinearInterpolate(B=@n3, Alpha=@n1.A), B=DynamicParameter() {parameterIndex=1, paramNames [SpriteSize,Param2,Param3,Param4]}.SpriteSize)
   OverridePivot
   "Normals (Optional)"
  ) {materialFunction=&Sprite}
  n4 = LinearInterpolate(
   A=VectorParameter() {parameterName=ColorLow, defaultValue=<rgba 1.7034 0.3111 5 1>}.RGB
   B=VectorParameter() {parameterName=ColorHigh, defaultValue=<rgba 0.2578 1.9367 5 1>}.RGB
   Alpha=Power(Base=@n2.R) {constExponent=3}
  )
  MaterialFunctionCall(Emissive=Multiply(A=Multiply(A=@n4, B=@n1.RGB), B=@n2.RGB)) {materialFunction=&MF_FX_EmissiveAdjustments}
 ]
}
