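# Niagara system for a rising, glowing, colored steam effect in the Secret Facility content; only its Smoke emitter renders, and two others are empty.
# Smoke spawns sprites above the origin, tinted randomly between purple and green-blue, moving upward.
# Curves grow each sprite, fade its alpha and boost color while draining green.
# User.MulitplierSmokeSpawnParticlesRate scales spawn rate and User.Scale sizes the spawn shape; other user parameters are unreferenced here.
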NiagaraSystem /SecretFacilityGameplay/Assets/FX/NS_SecretFacility_SteamInvis
dependencies {
 usedDataInterfaces [
  &NiagaraDataInterfaceCurve
  &NiagaraDataInterfaceVectorCurve
  &NiagaraDataInterfaceVector2DCurve
 ]
 usedDynamicInputs []
 usedModules [&SystemState]
}
emitters {
 Empty {rendererClasses [], renderers []}
 Empty001 {rendererClasses [], renderers []}
 Smoke {
  emitterUpdateScript [
   &EmitterState(Life Cycle Mode=::NewEnumerator1, Loop Delay=0.2, UseLoopDelay=true)
   &SpawnRate(SpawnRate=&Multiply_Float(A=15, B=$User.MulitplierSmokeSpawnParticlesRate), Use Spawn Probability=false)
  ]
  particleSpawnScript [
   &InitializeParticle(
    Color=&RandomRangeLinearColor(
     Minimum=&RandomRangeLinearColor(Minimum=<rgba 0.6604 0 1 0.7>, Maximum=<rgba 0.3798 0 1 0.7>)
     Maximum=&RandomRangeLinearColor(Minimum=<rgba 0 1 0.3879 0.7>, Maximum=<rgba 0 0.4678 1 0.7>)
    )
    Position Offset=<0 0 65>
    Sprite Rotation Mode=::NewEnumerator2
    Sprite Size Mode=::NewEnumerator1
    Uniform Sprite Size Max=35
    Uniform Sprite Size Min=25
    UsePositionOffset=true
    Write Material Random=true
   )
   &DynamicMaterialParameters(
    Index 0 Param 1=&RandomRangeFloat(Minimum=-0.001, Maximum=0.001)
    Index 0 Param 2=&RandomRangeFloat(Minimum=0, Maximum=32)
    Module.Index0Param0Write=true
    Module.Index0Param1Write=true
    Module.Index0Param2Write=false
    Module.Index0Param3Write=false
   )
   &ShapeLocation(
    MaskSpawnGroup=false
    Non Uniform Scale=&VectorFromFloat(Value=$User.Scale)
    Override Seed=false
    Sphere Radius=1
    UseOwnerScale=true
   )
   &AddVelocity(Velocity=<0 0 150>)
  ]
  particleUpdateScript [
   &ParticleState()
   &ScaleColor(
    Scale Alpha=&FloatFromCurve(FloatCurve=DataInterface() {propertyValues {Curve {keys [
         {arriveTangent=4.2554, interpMode=RCIM_Cubic, leaveTangent=4.2554, tangentMode=RCTM_User, time=0, value=0}
         {arriveTangent=0.8417, interpMode=RCIM_Cubic, leaveTangent=0.8417, time=0.264, value=0.9059}
         {arriveTangent=-1.2263, interpMode=RCIM_Cubic, leaveTangent=-1.2263, time=0.4887, value=0.4113}
         {arriveTangent=-0.0177, interpMode=RCIM_Cubic, leaveTangent=-0.0177, tangentMode=RCTM_User, time=1.0011, value=0.002}
        ]}}})
    Scale RGB=&VectorFromCurve(
     VectorCurve=DataInterface() {propertyValues {
       XCurve {keys [<at 0.0053 0.5>,<at 0.4026 1.3978>]}
       YCurve {keys [<at 0 1>,<at 0.4 0>]}
       ZCurve {keys [<at 0.0053 1.1046>,<at 0.4035 2.0225>]}
      }}
     Scale Curve=<5 5 5>
    )
    ScaleA=true
    ScaleRGB=true
   )
   &ScaleSpriteSize(
    Uniform Curve Scale=8
    Uniform Curve Sprite Scale=DataInterface() {propertyValues {Curve {keys [
        {arriveTangent=1.5814, interpMode=RCIM_Cubic, leaveTangent=1.5814, tangentMode=RCTM_User, time=0, value=0}
        {interpMode=RCIM_Cubic, leaveTangent=-0.4663, time=1, value=1}
       ]}}}
   )
   &SolveForcesAndVelocity(Clamp Velocity=false, Limit Acceleration=false)
  ]
  rendererClasses [&NiagaraSpriteRendererProperties]
  renderers [{rendererClass=&NiagaraSpriteRendererProperties}]
 }
}
summary {
 systemName=NS_SecretFacility_SteamInvis
 emitters [
  {emitterName=Empty, rendererClasses []}
  {emitterName=Empty001, rendererClasses []}
  {emitterName=Smoke, rendererClasses [&NiagaraSpriteRendererProperties]}
 ]
 userVariables [
  User.CloudSpawnParticlesRate {type=NiagaraFloat, defaultValue=10, description=""}
  User.CoreSpawnParticlesProbability {type=NiagaraFloat, defaultValue=1, description=""}
  User.CoreSpawnParticlesRate {type=NiagaraFloat, defaultValue=10, description=""}
  User.MulitplierSmokeSpawnParticlesRate {type=NiagaraFloat, defaultValue=1, description=""}
  User.Scale {type=NiagaraFloat, defaultValue=0, description=""}
  User.SmokeSpawnParticlesRate {type=NiagaraFloat, defaultValue=0, description=""}
  User.Timing_Lingering {type=NiagaraFloat, defaultValue=2, description=""}
 ]
}
