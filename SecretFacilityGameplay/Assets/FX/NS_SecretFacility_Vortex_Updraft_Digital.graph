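# Niagara system for a digital-styled vortex updraft in the Secret Facility, with upward-moving wind streaks, dust motes and glitch particles.
# Four GPU emitters spawn in owner-scaled cylinders, sharing system size and velocity multipliers; User.SpawnRate drives all but the fixed-rate Glitch_sprite meshes.
# WindStreaks and curl-noise Glitch_Hair render sprites; Dust_Motes001 and point-attracted Glitch_sprite render meshes.
# Camera-distance fades scale mote and hair size and glitch-mesh alpha.
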NiagaraSystem /SecretFacilityGameplay/Assets/FX/NS_SecretFacility_Vortex_Updraft_Digital
dependencies {
 usedDataInterfaces [&NiagaraDataInterfaceCurve, &NiagaraDataInterfaceCamera, &NiagaraDataInterfaceColorCurve]
 usedDynamicInputs [
  &SimulationPosition
  &Vector2DFromFloat
  &VectorFromFloat
 ]
 usedModules [
  &self:SystemSpawnScript.SystemScriptSource.SystemScriptGraph.NiagaraNodeAssignment_0.SetVariables_#1
  &self:SystemSpawnScript.SystemScriptSource.SystemScriptGraph.NiagaraNodeAssignment_1.SetVariables_#2
  &self:SystemSpawnScript.SystemScriptSource.SystemScriptGraph.NiagaraNodeAssignment_2.SetVariables_#3
  &SystemState
 ]
}
emitters {
 Dust_Motes001 {
  simTarget=GPUComputeSim
  emitterUpdateScript [
   &EmitterState(
    Life Cycle Mode=::NewEnumerator1
    Loop Duration=&RandomRangeFloat(Minimum=5, Maximum=10)
    Recalculate Duration Each Loop=true
    UseLoopDelay=false
   )
   &SpawnRate(SpawnRate=&Lerp_Float(A=0, B=100, Alpha=$NPC.FortniteNPC.IsPlayerIndoors), Use Spawn Probability=false) {enabled=false}
   &SpawnRate(SpawnRate=&Multiply_Float(A=15, B=$User.SpawnRate), Use Spawn Probability=false) {module=SpawnRate001}
  ]
  particleSpawnScript [
   &InitializeParticle(
    Color=&HueShiftLinearColor(
     Hue=&RandomRangeFloat(Minimum=0.9, Maximum=1)
     Saturation=&RandomRangeFloat(Minimum=0.25, Maximum=1)
     Value=1
     Linear Color=$System.BaseColor
    )
    Lifetime=&UniformRangedFloat(Minimum=3, Maximum=5)
    Mass Mode=::NewEnumerator1
    Mesh Scale Max=<0.1 0.1 2>
    Mesh Scale Min=<0.05 0.1 1>
    Mesh Scale Mode=::NewEnumerator2
    Sprite Rotation Mode=::NewEnumerator2
    Sprite Size Mode=::NewEnumerator3
    Uniform Sprite Size=5
    UsePositionOffset=false
    Write Material Random=true
   ) {module=InitializeParticle001}
   /SetVariables(Particles.MeshIndex=&RandomRangeInt(Minimum=0, Maximum=3)) {isSetParameters=true}
   &ShapeLocation(
    Cylinder Height=&Multiply_Float(A=2500, B=$System.LocationSizeMult)
    Cylinder Radius=&Multiply_Float(A=1500, B=$System.LocationSizeMult)
    MaskSpawnGroup=false
    Non Uniform Scale=<1 1 5>
    Offset=<0 0 6000>
    Offset Mode=::NewEnumerator0
    Override Seed=false
    Shape Primitive=::NewEnumerator1
    Surface Only=false
    UseOwnerScale=true
   )
   &AddVelocity(Velocity=&UniformRangedVector(Minimum=<-10 -10 600>, Maximum=<10 10 1000>))
   &DynamicMaterialParameters(
    Index 0 Param 1=2
    Index 0 Param 2=50
    Index 0 Param 3=200
    Module.Param0WriteEnabled=true
    Module.Param1WriteEnabled=true
    Module.Param2WriteEnabled=true
    Module.Param3WriteEnabled=true
   ) {module=DynamicMaterialParameters001}
   &InitialMeshOrientation(Mesh Axis To Orient=<0 0 1>, Mesh Orientation Mode=::NewEnumerator2, Use Rotation Vector=false)
   /SetVariables(Particles.Velocity=&Multiply_Vector(A=$Particles.Velocity, B=$System.Vel Mul)) {isSetParameters=true}
  ]
  particleUpdateScript [
   &ParticleState()
   &ScaleMeshSize(Scale Factor=&Multiply_VectorByFloat(
     Vector=<1 1 1>
     Float=&Multiply_Float(
      A=1
      B=&Lerp_Float(
       A=0.2
       B=1
       Alpha=/Niagara/DynamicInputs/Camera/CameraFade(
        Camera Query:NiagaraDataInterfaceCamera=DataInterface()
        Fade Offset:NiagaraFloat=500
        Fade Distance:NiagaraFloat=500
        Use Owner Position:NiagaraBool=true
       )
      )
     )
    ))
   &Drag(Drag=0.5, Use Linear Drag=true, Use Rotational Drag=false)
   &GravityForce(Gravity=<0 0 -15>)
   &ScaleColor(Scale Alpha=0.002, Scale RGB=<5 5 5>, ScaleA=true, ScaleRGB=true)
   &SolveForcesAndVelocity(Clamp Velocity=true, Limit Acceleration=false, Speed Limit=500) {module=SolveForcesAndVelocity001}
   &DynamicMaterialParameters(Module.Param0WriteEnabled=true, Module.Param1WriteEnabled=true, Module.Param2WriteEnabled=true, Module.Param3WriteEnabled=true)
  ]
  rendererClasses [&NiagaraMeshRendererProperties]
  renderers [
   {rendererClass=&NiagaraMeshRendererProperties, rendererIndex=0}
   {rendererClass=&NiagaraMeshRendererProperties, rendererIndex=1}
  ]
 }
 Glitch_Hair {
  simTarget=GPUComputeSim
  emitterUpdateScript [
   &EmitterState()
   &SpawnRate(SpawnRate=&Multiply_Float(A=10, B=$User.SpawnRate), Use Spawn Probability=false) {module=SpawnRate001}
  ]
  particleSpawnScript [
   &InitializeParticle(
    Color=&RandomRangeLinearColor(Minimum=<rgba 1 0.2064 0.2064 1>, Maximum=<rgba 1 0.6613 0.0971 1>)
    Lifetime Mode=::NewEnumerator1
    Sprite Size Min=<5 1>
    Sprite Size Mode=::NewEnumerator2
    UsePositionOffset=true
    Write Material Random=true
   ) {module=InitializeParticle001}
   &CameraOffset(Camera Offset Amount=3) {module=CameraOffset001}
   &ShapeLocation(
    Cylinder Height=&Multiply_Float(A=2500, B=$System.LocationSizeMult)
    Cylinder Radius=&Multiply_Float(A=1500, B=$System.LocationSizeMult)
    MaskSpawnGroup=false
    Non Uniform Scale=<1 1 5>
    Offset=<0 0 6000>
    Offset Mode=::NewEnumerator0
    Override Seed=false
    Shape Primitive=::NewEnumerator1
    Surface Only=false
    UseOwnerScale=true
   )
   &AddVelocity(Velocity=&UniformRangedVector(Minimum=<-10 -10 200>, Maximum=<10 10 250>))
   /SetVariables(Particles.Velocity=&Multiply_Vector(A=$Particles.Velocity, B=$System.Vel Mul)) {isSetParameters=true}
  ]
  particleUpdateScript [
   &ParticleState()
   &Drag(Use Linear Drag=true, Use Rotational Drag=false)
   &CurlNoiseForce(
    Bias Noise=false
    Curl Noise Cone Mask Angle=90
    Curl Noise Cone Mask Falloff Angle=90
    Mask Curl Noise=true
    Noise Frequency=120
    Noise Strength=350
    Pan Noise=false
   )
   &SolveForcesAndVelocity(Clamp Velocity=true, Limit Acceleration=false, Speed Limit=500)
   &ScaleSpriteSize(
    Scale Sprite Size Mode=::NewEnumerator0
    Uniform Scale Factor=&Multiply_Float(
     A=&FloatFromCurve(
      FloatCurve=DataInterface() {propertyValues {Curve {keys [
          {arriveTangent=5.4245, time=0, value=5}
          {arriveTangent=-3.9614, time=0.0963, value=8.5355}
          {arriveTangent=-22.2662, leaveTangent=-22.2662, time=0.2, value=4}
          <at 0.3 4>
          <at 0.4 11.5>
          {arriveTangent=-0.0405, leaveTangent=-0.0405, time=1, value=0}
         ]}}}
      Scale Curve=0.1
     )
     B=&Multiply_Float(
      A=1
      B=&Lerp_Float(
       A=0.2
       B=1
       Alpha=&CameraFade(
        Camera Query:NiagaraDataInterfaceCamera=DataInterface()
        Fade Offset:NiagaraFloat=500
        Fade Distance:NiagaraFloat=500
        Use Owner Position:NiagaraBool=true
       )
      )
     )
    )
   )
   &DynamicMaterialParameters(
    Index 0 Param 1=&FloatFromCurve(
     FloatCurve=DataInterface() {propertyValues {Curve {keys [
         {arriveTangent=4.1785, interpMode=RCIM_Cubic, leaveTangent=4.1785, tangentMode=RCTM_User, time=0, value=0}
         {arriveTangent=0.9995, interpMode=RCIM_Cubic, leaveTangent=0.9995, tangentMode=RCTM_User, time=0.2223, value=0.6825}
         {arriveTangent=-0.0405, interpMode=RCIM_Cubic, leaveTangent=-0.0405, tangentMode=RCTM_User, time=1, value=1}
        ]}}}
     Scale Curve=0.25
    )
    Index 0 Param 2=&RandomRangeFloat(Minimum=-1, Maximum=0.1)
    Index 0 Param 3=&Multiply_Float(
     A=&FloatFromCurve(
      FloatCurve=DataInterface() {propertyValues {Curve {keys [<at 0 1 RCIM_Cubic>, {arriveTangent=-1.8781, interpMode=RCIM_Cubic, leaveTangent=-1.8781, tangentMode=RCTM_User, time=1, value=0}]}}}
      Scale Curve=&RandomRangeFloat(Minimum=0.25, Maximum=1)
     )
     B=0.5
    )
    Index 0 Param 4=&FloatFromCurve(FloatCurve=DataInterface() {propertyValues {Curve {keys [<at 0 0.1>,<at 1 1>]}}})
    Module.Param0WriteEnabled=true
    Module.Param1WriteEnabled=true
    Module.Param2WriteEnabled=true
    Module.Param3WriteEnabled=true
   ) {module=DynamicMaterialParameters002}
   &ScaleColor(
    Scale Alpha=&FloatFromCurve(
     FloatCurve=DataInterface() {propertyValues {Curve {keys [{interpMode=RCIM_Cubic, tangentMode=RCTM_User, time=0, value=0}, <at 0.1 1>, {interpMode=RCIM_Cubic, tangentMode=RCTM_User, time=1, value=0}]}}}
     Scale Curve=0.5
    )
    ScaleA=true
    ScaleRGB=false
   ) {module=ScaleColor003}
   /SetVariables(Particles.VisibilityTag=&MakeCustomIntFromBool(Bool=$User.bisHeadHidden, True INT=1, False INT=0)) {isSetParameters=true}
  ]
  rendererClasses [&NiagaraSpriteRendererProperties]
  renderers [{rendererClass=&NiagaraSpriteRendererProperties}]
 }
 Glitch_sprite {
  simTarget=GPUComputeSim
  emitterUpdateScript [
   &EmitterState()
   &SpawnRate(SpawnRate=1500, Use Spawn Probability=false)
  ]
  particleSpawnScript [
   &InitializeParticle(
    Color=<rgba 5 5 5 0.175>
    Lifetime Max=1.5
    Lifetime Mode=::NewEnumerator1
    Mesh Scale Max=<0.75 0.75 5>
    Mesh Scale Min=<0.25 0.25 0.5>
    Mesh Scale Mode=::NewEnumerator2
    Position Offset=<0 0 100>
    UsePositionOffset=true
    Write Material Random=true
   )
   &CameraOffset(Camera Offset Amount=50)
   &ShapeLocation(
    Cylinder Height=&Multiply_Float(A=2500, B=$System.LocationSizeMult)
    Cylinder Radius=&Multiply_Float(A=1500, B=$System.LocationSizeMult)
    MaskSpawnGroup=false
    Non Uniform Scale=<1 1 5>
    Offset=<0 0 6000>
    Offset Mode=::NewEnumerator0
    Override Seed=false
    Shape Primitive=::NewEnumerator1
    Surface Only=false
    UseOwnerScale=true
   )
   &SystemLocation(MaskSpawnGroup=false) {enabled=false}
   &AddVelocity(Velocity=&UniformRangedVector(Minimum=<-10 -10 300>, Maximum=<10 10 600>))
   &InitialMeshOrientation(Mesh Axis To Orient=<0 0 1>, Mesh Orientation Mode=::NewEnumerator2, Use Rotation Vector=false)
   /SetVariables(Particles.Velocity=&Multiply_Vector(A=$Particles.Velocity, B=$System.Vel Mul)) {isSetParameters=true}
  ]
  particleUpdateScript [
   &ParticleState()
   &ScaleMeshSize(Scale Factor=&Multiply_VectorByFloat(
     Vector=<1 1 1>
     Float=&Multiply_Float(
      A=0.2
      B=&FloatFromCurve(
       FloatCurve=DataInterface() {propertyValues {Curve {keys [
           {arriveTangent=4.1785, time=0, value=1}
           <at 0.5 1.1>
           {arriveTangent=-0.0405, interpMode=RCIM_Cubic, leaveTangent=-0.0405, tangentMode=RCTM_User, time=1, value=1.5}
          ]}}}
       Scale Curve=5
      )
     )
    ))
   /Niagara/Modules/Update/Forces/PointAttractionForce(
    Attraction Radius=50000
    AttractorPosition:Vector3f=&Add_Vector(A=/Niagara/DynamicInputs/Helpers/SimulationPosition(), B=<0 0 15000>)
    Module.Kill Within Radius=false
    Module.Use Falloff=true
   )
   /Niagara/Modules/Update/Forces/CurlNoiseForce(Noise Frequency=0.1, Noise Strength=1, Pan Noise=true, Pan Noise Field=<0 0 2>)
   &Drag(Use Linear Drag=true, Use Rotational Drag=true)
   &SolveForcesAndVelocity(Clamp Velocity=false, Limit Acceleration=false)
   &DynamicMaterialParameters(Module.Param0WriteEnabled=true, Module.Param1WriteEnabled=true, Module.Param2WriteEnabled=true, Module.Param3WriteEnabled=true)
   &ScaleColor(
    Scale Alpha=&Multiply_Float(
     A=&FloatFromCurve(
      FloatCurve=DataInterface() {propertyValues {Curve {keys [
          {arriveTangent=9.0725, interpMode=RCIM_Cubic, leaveTangent=9.0725, tangentMode=RCTM_User, time=0, value=0}
          {arriveTangent=-2.5249, interpMode=RCIM_Cubic, leaveTangent=-2.5249, tangentMode=RCTM_User, time=0.5, value=0.9847}
          {arriveTangent=-0.0405, interpMode=RCIM_Cubic, leaveTangent=-0.0405, tangentMode=RCTM_User, time=1, value=0}
         ]}}}
      Scale Curve=&RandomRangeFloat(Minimum=0.2, Maximum=1)
     )
     B=&Multiply_Float(
      A=0.02
      B=&CameraFade(
       Camera Query:NiagaraDataInterfaceCamera=DataInterface()
       Fade Offset:NiagaraFloat=500
       Fade Distance:NiagaraFloat=1500
       Use Owner Position:NiagaraBool=true
      )
     )
    )
    Scale RGB=&Multiply_VectorByFloat(
     Vector=&MakeVectorFromLinearColorRGB(Color=&ColorFromCurve(
       ColorCurve=DataInterface() {propertyValues {
         AlphaCurve {keys []}
         BlueCurve {keys [<at 0.4803 0>,<at 0.9765 0>]}
         GreenCurve {keys [<at 0.4803 0.4421>,<at 0.9765 4.1667>]}
         RedCurve {keys [<at 0.4803 50>,<at 0.9765 50>]}
        }}
       CurveIndex=&RandomRangeFloat(Minimum=0, Maximum=1)
      ))
     Float=&FloatFromCurve(FloatCurve=DataInterface() {propertyValues {Curve {keys [{arriveTangent=-2.0684, interpMode=RCIM_Cubic, leaveTangent=-2.0684, tangentMode=RCTM_User, time=0, value=2}, <at 1 0>]}}})
    )
    ScaleA=true
    ScaleRGB=true
   )
  ]
  rendererClasses [&NiagaraMeshRendererProperties]
  renderers [{rendererClass=&NiagaraMeshRendererProperties, rendererIndex=0}]
 }
 WindStreaks {
  simTarget=GPUComputeSim
  emitterUpdateScript [
   &EmitterState()
   &SpawnRate(SpawnRate=&Multiply_Float(A=$User.SpawnRate, B=5), Use Spawn Probability=false) {module=SpawnRate001}
  ]
  particleSpawnScript [
   &InitializeParticle(
    Color=$System.BaseColor
    Lifetime=&UniformRangedFloat(Minimum=1, Maximum=3)
    Mass Mode=::NewEnumerator1
    Sprite Size=&UniformRangedVector2D(Minimum=<100 500>, Maximum=<200 800>)
    Sprite Size Mode=::NewEnumerator4
    UsePositionOffset=false
    Write Material Random=true
   ) {module=InitializeParticle001}
   /Niagara/Modules/Update/SubUV/SubUVAnimation(SubUV Animation Mode=::NewEnumerator1)
   &AddVelocity(Velocity=&UniformRangedVector(Minimum=<-10 -10 100>, Maximum=<10 10 300>))
   &AccelerationForce(Acceleration=<0 0 500>)
   &ApplyInitialForces()
   &SystemLocation(MaskSpawnGroup=false)
   &ShapeLocation(
    Cylinder Height=&Multiply_Float(A=8000, B=$System.LocationSizeMult)
    Cylinder Radius=&Multiply_Float(A=1500, B=$System.LocationSizeMult)
    MaskSpawnGroup=false
    Override Seed=false
    Shape Primitive=::NewEnumerator1
    Surface Only=false
    Surface Only Band Thickness=0.25
    UseOwnerScale=true
   )
   /SetVariables(Particles.Velocity=&Multiply_VectorByFloat(Vector=&Multiply_Vector(A=$Particles.Velocity, B=$System.Vel Mul), Float=0.7)) {isSetParameters=true}
  ]
  particleUpdateScript [
   &ParticleState()
   &PointForce(
    Force Origin=&Add_Vector(A=<0 0 500>, B=<0 0 500>)
    Force Strength=500
    Module.Use Falloff=true
    Module.UseLinearForce=false
   )
   &SolveForcesAndVelocity(Clamp Velocity=false, Limit Acceleration=false)
   &DynamicMaterialParameters(
    Index 0 Param 1=1
    Index 0 Param 2=&FloatFromCurve(FloatCurve=DataInterface() {propertyValues {Curve {keys [<at 0 0.1>,<at 1 0.25>]}}}, Scale Curve=3)
    Index 0 Param 3=&FloatFromCurve(FloatCurve=DataInterface() {propertyValues {Curve {keys [<at 0 0.8>,<at 1 1>]}}})
    Module.Param0WriteEnabled=true
    Module.Param1WriteEnabled=true
    Module.Param2WriteEnabled=true
    Module.Param3WriteEnabled=true
   )
   &ScaleColor(
    Scale Alpha=&FloatFromCurve(
     FloatCurve=DataInterface() {propertyValues {Curve {keys [
         {interpMode=RCIM_Cubic, tangentMode=RCTM_User, time=0, value=0}
         {arriveTangent=0.002, interpMode=RCIM_Cubic, leaveTangent=0.002, tangentMode=RCTM_User, time=0.6, value=1.05}
         {interpMode=RCIM_Cubic, tangentMode=RCTM_User, time=1.0011, value=0.002}
        ]}}}
     Scale Curve=0.25
    )
    Scale RGB=&MakeVectorFromLinearColorRGB(Color=<rgba 1 1 1 1>)
    ScaleA=true
    ScaleRGB=true
   )
  ]
  rendererClasses [&NiagaraSpriteRendererProperties]
  renderers [{rendererClass=&NiagaraSpriteRendererProperties}]
 }
}
summary {
 systemName=NS_SecretFacility_Vortex_Updraft_Digital
 emitters [
  {emitterName=WindStreaks, simTarget=GPUComputeSim, rendererClasses [&NiagaraSpriteRendererProperties]}
  {emitterName=Dust_Motes001, simTarget=GPUComputeSim, rendererClasses [&NiagaraMeshRendererProperties]}
  {emitterName=Glitch_sprite, simTarget=GPUComputeSim, rendererClasses [&NiagaraMeshRendererProperties]}
  {emitterName=Glitch_Hair, simTarget=GPUComputeSim, rendererClasses [&NiagaraSpriteRendererProperties]}
 ]
 userVariables [
  User.bIsFrontEnd {type=NiagaraBool}
  User.bIsFrontEndPreview {type=NiagaraBool}
  User.bisHeadHidden {type=NiagaraBool}
  User.CylinderRadius {type=NiagaraFloat, defaultValue=0, description=""}
  User.GlitchSpawnProbability {type=NiagaraFloat, defaultValue=1, description=""}
  User.SpawnRate {type=NiagaraFloat, defaultValue=50, description=""}
 ]
}
