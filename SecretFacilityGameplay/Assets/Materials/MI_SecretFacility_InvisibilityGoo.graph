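# Material for the Secret Facility invisibility goo, going by its name; it is a base Material despite the MI_ prefix.
# It uses translucent blending with unlit shading and has automatic usage flag setting in the editor turned off.
# The graph holds no expressions, so the file shows no logic driving color, opacity or any invisibility effect.
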Material /SecretFacilityGameplay/Assets/Materials/MI_SecretFacility_InvisibilityGoo
graph {
 expressions {}
 properties {bAutomaticallySetUsageInEditor=false, blendMode=BLEND_Translucent, shadingModel=MSM_Unlit, usedShadingModels=Unlit}
}
