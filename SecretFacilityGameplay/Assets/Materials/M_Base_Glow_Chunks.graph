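# M_Base_Glow_Chunks is an unlit translucent base material for Niagara mesh particles, named for glowing chunks.
# Its only graph logic feeds particle color alpha into the shared MF_DitherBayerNoTex function, a Bayer dither that by its name needs no texture.
# The graph shows no connection from that function to any material output, and no color or emissive logic.
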Material /SecretFacilityGameplay/Assets/Materials/M_Base_Glow_Chunks
graph {
 properties {
  bAutomaticallySetUsageInEditor=false
  bUsedWithNiagaraMeshParticles=true
  blendMode=BLEND_Translucent
  shadingModel=MSM_Unlit
  usedShadingModels=Unlit
 }
 nodes [MaterialFunctionCall!(In=ParticleColor().A, Iterations, Scale) {materialFunction=/Game/Packages/Fortress_SharedMaterials/Base_Material_Functions/MF_DitherBayerNoTex}]
}
