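# Translucent, unlit, two-sided Niagara material for the VitalPsych contrail trail, drawing animated neon strips and circuitry.
# Emissive randomly picks one of two gradients per particle, then blends in LineColor through scrolling circuit lines.
# Opacity combines panning large and small strips with noise-broken circuitboard, shaped by UV fades and particle alpha.
# Dynamic parameters toggle layers and offset strip tiling and speed; UVPanTime drives several panners.
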Material /SecretFacilityGameplay/Assets/Materials/M_Contrail_VitalPsych_Trail_Sprite
graph {
 properties {
  bCastRayTracedShadows=false
  bEnableResponsiveAA=true
  bUsedWithNiagaraMeshParticles=true
  bUsedWithNiagaraRibbons=true
  bUsedWithNiagaraSprites=true
  blendMode=BLEND_Translucent
  pixelDepthOffsetMode=PDOM_Legacy
  shadingModel=MSM_Unlit
  twoSided=true
  usedShadingModels=Unlit
 }
 nodes [
  n1 = Power(Base=ComponentMask(None=TextureCoordinate()) {g=true}) {constExponent=3}
  n2 = LinearInterpolate(
   A=VectorParameter() {parameterName=ColorB, defaultValue=<rgba 0.0314 0 0.5 1>}.RGB
   B=VectorParameter() {parameterName=ColorA, defaultValue=<rgba 3 0.0969 0.6494 1>}.RGB
   Alpha=@n1
  )
  n3 = LinearInterpolate(
   A=VectorParameter() {parameterName=ColorC, defaultValue=<rgba 0 0.2431 1 1>}.RGB
   B=VectorParameter() {parameterName=ColorD, defaultValue=<rgba 0 1.9922 2 1>}.RGB
   Alpha=@n1
  )
  n4 = Round(None=ParticleRandom())
  n5 = Panner(Coordinate=Multiply(A=TextureCoordinate() {uTiling=0.5, vTiling=3}, B=ParticleRandom())) {speedY=0.25}
  n6 = ComponentMask(None=TextureCoordinate()) {g=true}
  n7 = TextureCoordinate()
  n8 = NamedRerouteUsage()
  n9 = Multiply(A=TextureSample(UVs=Panner(Coordinate=@n7, Time=@n8) {speedY=0.75}) {texture=/Game/Global/Textures/Default/Masks_and_Gradients/CylGradient}.RGB) {constB=4}
  n10 = Multiply(
   A=Multiply(A=TextureSample(UVs=@n5) {texture=/Game/Effects/Fort_Effects/Textures/T_FX_Circuitboard}.RGB, B=Power(Base=@n6) {constExponent=1.5})
   B=Saturate(None=Clamp(None=@n9) {minDefault=0.2})
  )
  n11 = DynamicParameter() {parameterIndex=1, defaultValue=<rgba 0 0 0 0>, paramNames [VisToggle_LgStrips, VisToggle_OrangeStrips, Param3, Param4]}
  MaterialFunctionCall(Emissive=LinearInterpolate(
    A=LinearInterpolate(A=@n2, B=@n3, Alpha=@n4)
    B=VectorParameter() {parameterName=LineColor, defaultValue=<rgba 1 0.3344 0 1>}.RGB
    Alpha=LinearInterpolate(A=@n10, Alpha=@n11.VisToggle_OrangeStrips) {constB=0}
   )) {materialFunction=&MF_FX_EmissiveAdjustments}
  NamedRerouteDeclaration(None=ScalarParameter() {parameterName=UVPanTime}) {name=UV Pan Time, nodeColor=<rgba 0 1 0.9063 1>}
  n12 = TextureCoordinate()
  n13 = MaterialFunctionCall(In=Power(Base=ComponentMask(None=@n12) {g=true}) {constExponent=1}, Contrast=Constant() {r=0.1}) {materialFunction=&CheapContrast}
  n14 = ComponentMask(None=TextureCoordinate()) {g=true}
  n15 = Multiply(A=MaterialFunctionCall(In=OneMinus(None=@n14), Contrast=Constant() {r=0.001}) {materialFunction=&CheapContrast}.Result) {constB=6}
  n16 = Constant()
  n17 = DynamicParameter() {defaultValue=<rgba 0 0 0 0>, paramNames [SmStrip_TileY, LgStrip_Erode, SmStrip_PanSpeed, SmStrip_TileX]}
  n18 = Add(A=@n17.SmStrip_TileY, B=ScalarParameter() {defaultValue=1, parameterName=LgStrip_PanSpeed})
  n19 = TextureSample(UVs=Panner(Coordinate=TextureCoordinate() {uTiling=0.9713}, Time=NamedRerouteUsage(), Speed=AppendVector(A=@n16, B=@n18)) {speedX=-3}) {samplerType=SAMPLERTYPE_LinearColor, texture=/Game/Effects/Fort_Effects/Effects/Characters/Generic/ResIn/DiffusedStylizedGridAnimOffsets}
  n20 = Multiply(A=Saturate(None=Multiply(A=@n19.RGB) {constB=2}))
  n21 = TextureCoordinate()
  n22 = Constant() {r=0.25}
  n23 = Constant() {r=1}
  n24 = Constant() {r=7.6698}
  n25 = Panner(
   Coordinate=MaterialFunctionCall(
    UVs=MaterialFunctionCall(UVs=@n21, "Rotation Angle (0-1)"=@n22) {materialFunction=&CustomRotator}.Rotated Values
    Texture Scale=AppendVector(A=@n23, B=@n24)
   ) {materialFunction=&ScaleUVsByCenter}.UVs
   Time=NamedRerouteUsage()
  ) {speedX=-0.1}
  n26 = AppendVector(A=Constant() {r=1}, B=Constant() {r=1.9558})
  n27 = Power(Base=TextureSample(UVs=Multiply(A=@n25, B=@n26)) {texture=/Game/Effects/Fort_Effects/Textures/Tile/Tests/T_Tile_Panner_Mask}.RGB) {constExponent=0.4}
  n28 = DynamicParameter() {parameterIndex=1, defaultValue=<rgba 0 0 0 0>, paramNames [VisToggle_LgStrips, VisToggle_OrangeStrips, Param3, Param4]}
  n29 = LinearInterpolate(A=Add(A=Saturate(None=@n20), B=Saturate(None=@n27)), Alpha=@n28.VisToggle_LgStrips) {constB=0}
  n30 = TextureCoordinate()
  n31 = AppendVector(A=Constant() {r=2}, B=Constant() {r=0.4})
  n32 = DynamicParameter() {defaultValue=<rgba 0 0 0 0>, paramNames [SmStrip_TileY, LgStrip_Erode, SmStrip_PanSpeed, SmStrip_TileX]}
  n33 = Add(A=@n32.SmStrip_TileX, B=ScalarParameter() {defaultValue=2, parameterName=SmStrip_TileX})
  n34 = Add(A=ScalarParameter() {defaultValue=0.4, parameterName=SmStrip_TileY}, B=@n32.SmStrip_TileY)
  n35 = ScalarParameter() {defaultValue=2, parameterName=SmStrip_PanSpeed}
  n36 = DynamicParameter() {defaultValue=<rgba 0 0 0 0>, paramNames [SmStrip_TileY, LgStrip_Erode, SmStrip_PanSpeed, SmStrip_TileX]}
  n37 = Panner(
   Coordinate=Multiply(
    A=MaterialFunctionCall(UVs=@n30, Texture Scale=@n31) {materialFunction=&ScaleUVsByCenter}.UVs
    B=AppendVector(A=@n33, B=@n34)
   )
   Time=NamedRerouteUsage()
   Speed=AppendVector(A=Constant(), B=Add(A=@n35, B=@n36.SmStrip_PanSpeed))
  ) {speedX=-0.1}
  n38 = Constant() {r=0.75}
  n39 = MaterialFunctionCall(
   In=TextureSample(UVs=MaterialFunctionCall(UVs=@n37, "Rotation Angle (0-1)"=@n38) {materialFunction=&CustomRotator}.Rotated Values) {texture=&T_Tile_Panner_Mask}.RGB
   Contrast=Constant() {r=0.1}
  ) {materialFunction=&CheapContrast}
  n40 = TextureCoordinate()
  n41 = Power(Base=OneMinus(None=ComponentMask(None=@n40) {r=true})) {constExponent=1}
  n42 = TextureCoordinate() {uTiling=0.5, vTiling=0.5}
  n43 = Constant()
  n44 = TextureSample(UVs=Panner(Coordinate=TextureCoordinate() {uTiling=0.2, vTiling=0.2}) {speedY=0.5}) {samplerType=SAMPLERTYPE_LinearGrayscale, texture=/Game/Effects/Fort_Effects/Textures/Tile/T_CellNoise_Tile_M}
  n45 = Constant() {r=1}
  n46 = Multiply(
   A=TextureSample(UVs=MaterialFunctionCall(UVs=@n42, "Rotation Angle (0-1)"=@n43) {materialFunction=&CustomRotator}.Rotated Values) {texture=&T_FX_Circuitboard}.G
   B=OneMinus(None=MaterialFunctionCall(In=@n44.RGB, Contrast=@n45) {materialFunction=&CheapContrast}.Result)
  )
  n47 = OneMinus(None=Subtract(A=Add(A=@n29, B=@n39.Result), B=Multiply(A=@n41, B=@n46)))
  n48 = OneMinus(None=ComponentMask(None=TextureCoordinate()) {r=true})
  n49 = Multiply(
   A=TextureCoordinate()
   B=AppendVector(A=ScalarParameter() {defaultValue=0.4, parameterName=Param_14}, B=ScalarParameter() {defaultValue=0.4, parameterName=Param_15})
  )
  n50 = Constant()
  n51 = TextureSample(UVs=Panner(Coordinate=TextureCoordinate() {uTiling=0.2, vTiling=0.2}) {speedY=0.5}) {samplerType=SAMPLERTYPE_LinearGrayscale, texture=&T_CellNoise_Tile_M}
  n52 = Constant() {r=1}
  n53 = Multiply(
   A=TextureSample(UVs=MaterialFunctionCall(UVs=@n49, "Rotation Angle (0-1)"=@n50) {materialFunction=&CustomRotator}.Rotated Values) {texture=&T_FX_Circuitboard}.G
   B=OneMinus(None=MaterialFunctionCall(In=@n51.RGB, Contrast=@n52) {materialFunction=&CheapContrast}.Result)
  )
  n54 = DynamicParameter() {parameterIndex=1, defaultValue=<rgba 0 0 0 0>, paramNames [VisToggle_LgStrips, VisToggle_OrangeStrips, Param3, Param4]}
  n55 = Add(
   A=Subtract(A=Saturate(None=@n15), B=Saturate(None=@n47))
   B=Multiply(A=Power(Base=@n48) {constExponent=3}, B=LinearInterpolate(A=@n53, Alpha=@n54.VisToggle_LgStrips) {constB=0})
  )
  Reroute(None=Multiply(A=ParticleColor().A, B=Multiply(A=@n13.Result, B=@n55)))
  Subtract(A=ComponentMask(None=TextureCoordinate()) {r=true}, B=@n20)
 ]
}
